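# Replay highlight list item widget, a fixed-width row that shows an icon, a title and a clock-marked duration for one highlight.
# Layout is a bordered horizontal box; the icon, title and duration widgets are inherited from FortReplayHighlightItemBase, with placeholder content here.
# On Clicked calls the base class's HandlePlayItemClicked, the only wired logic in the graph.
# The inherited InputActionWidget is not placed in the layout.
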WidgetBlueprint /FortniteCoreUI/UI/Replay/Highlights/FortReplayHighlightItem
graphs [EventGraph {nodes [
   Event On Clicked()
   > HandlePlayItemClicked(self:Fort Replay Highlight Item Base Object Reference)
   Event PreConstruct()
   Event Construct()
   Event Tick()
  ]}]
variables []
widgets {
 info {inheritedWidgetCount=4, parentClass=/Script/FortniteUI.FortReplayHighlightItemBase, rootWidgetClass=&SizeBox, widgetCount=12}
 widgets [
  SizeBox_0 {widgetClassPath=&SizeBox, widget=SizeBox_0, properties {bOverride_WidthOverride=true, widthOverride=630}}
  CommonBorder_0 {
   widgetClassPath=&CommonBorder
   parent=SizeBox_0
   widget=CommonBorder_0
   properties {
    background {
     bIsDynamicallyLoaded=false
     drawAs=Image
     imageType=NoImage
     mirroring=NoMirror
     resourceName=None
     resourceObject=/FortUILibrary/Styles/Border/Materials/M_AthenaStore_OfferHeaderBacking
     tiling=NoTile
     imageSize=<64 64>
     margin=<ltrb 0 0 0 0>
     outlineSettings {
      bUseBrushTransparency=false
      roundingType=HalfHeightRadius
      width=0
      color {colorUseRule=UseColor_Specified, specifiedColor=<rgba 0 0 0 0>}
      cornerRadii=<0 0 0 0>
     }
     tintColor {colorUseRule=UseColor_Specified, specifiedColor=<rgba 1 1 1 1>}
     uVRegion {bIsValid=false, max=<0 0>, min=<0 0>}
    }
    brushColor=<rgba 1 1 1 0.5>
    padding=<ltrb 4 6 4 6>
   }
   slot {class=SizeBoxSlot}
  }
  HorizontalBox_1 {
   widgetClassPath=&HorizontalBox
   parent=CommonBorder_0
   widget=HorizontalBox_1
   slot {class=BorderSlot, properties {padding=<ltrb 4 6 4 6>}}
  }
  IconSB {
   widgetClassPath=&SizeBox
   parent=HorizontalBox_1
   widget=IconSB
   properties {bOverride_HeightOverride=true, bOverride_WidthOverride=true, heightOverride=32, widthOverride=32}
   slot {class=HorizontalBoxSlot, properties {horizontalAlignment=HAlign_Center, verticalAlignment=VAlign_Center, padding=<ltrb 20 0 15 0>}}
  }
  Image_Icon {
   widgetClassPath=&Image
   bInherited=true
   bIsVariable=true
   parent=IconSB
   widget=Image_Icon
   properties {brush {
     bIsDynamicallyLoaded=false
     drawAs=Image
     imageType=NoImage
     mirroring=NoMirror
     resourceName=None
     resourceObject=/Game/UI/Foundation/Textures/Icons/Replay/T-Icon-Replay-CameraSettings-M
     tiling=NoTile
     imageSize=<32 32>
     margin=<ltrb 0 0 0 0>
     outlineSettings {
      bUseBrushTransparency=false
      roundingType=HalfHeightRadius
      width=0
      color {colorUseRule=UseColor_Specified, specifiedColor=<rgba 0 0 0 0>}
      cornerRadii=<0 0 0 0>
     }
     tintColor {colorUseRule=UseColor_Specified, specifiedColor=<rgba 1 1 1 1>}
     uVRegion {bIsValid=false, max=<0 0>, min=<0 0>}
    }}
   slot {class=SizeBoxSlot}
  }
  TextSizeBox {
   widgetClassPath=&SizeBox
   parent=HorizontalBox_1
   widget=TextSizeBox
   properties {bOverride_WidthOverride=true, widthOverride=425}
   slot {class=HorizontalBoxSlot, properties {horizontalAlignment=HAlign_Center, verticalAlignment=VAlign_Center}}
  }
  TextBlock_Title {
   widgetClassPath=&TextBlock
   bInherited=true
   bIsVariable=true
   parent=TextSizeBox
   widget=TextBlock_Title
   properties {
    text=Game Summary
    font {
     bForceMonospaced=false
     bMaterialIsStencil=false
     fontMaterial=null
     fontObject=&BurbankSmall
     letterSpacing=0
     monospacedWidth=1
     size=18
     skewAmount=0
     typefaceFontName=Bold
     outlineSettings {
      bApplyOutlineToDropShadows=false
      bMiteredCorners=false
      bSeparateFillAlpha=false
      outlineMaterial=null
      outlineSize=0
      outlineColor=<rgba 0 0 0 1>
     }
    }
   }
   slot {class=SizeBoxSlot, properties {padding=<ltrb 0 1 0 0>}}
  }
  HorizontalBox_2 {
   widgetClassPath=&HorizontalBox
   parent=HorizontalBox_1
   widget=HorizontalBox_2
   slot {class=HorizontalBoxSlot, properties {horizontalAlignment=HAlign_Left, verticalAlignment=VAlign_Top, padding=<ltrb 5 0 0 0>}}
  }
  ClockSB {
   widgetClassPath=&SizeBox
   parent=HorizontalBox_2
   widget=ClockSB
   properties {bOverride_HeightOverride=true, bOverride_WidthOverride=true, heightOverride=28, widthOverride=28}
   slot {class=HorizontalBoxSlot, properties {padding=<ltrb 0 2 10 2>}}
  }
  ClockImage {
   widgetClassPath=&Image
   bIsVariable=true
   parent=ClockSB
   widget=ClockImage
   properties {brush {
     bIsDynamicallyLoaded=false
     drawAs=Image
     imageType=NoImage
     mirroring=NoMirror
     resourceName=None
     resourceObject=/Game/UI/Foundation/Textures/Icons/Replay/T-Icon-Replay-Timeline-M
     tiling=NoTile
     imageSize=<28 28>
     margin=<ltrb 0 0 0 0>
     outlineSettings {
      bUseBrushTransparency=false
      roundingType=HalfHeightRadius
      width=0
      color {colorUseRule=UseColor_Specified, specifiedColor=<rgba 0 0 0 0>}
      cornerRadii=<0 0 0 0>
     }
     tintColor {colorUseRule=UseColor_Specified, specifiedColor=<rgba 1 1 1 1>}
     uVRegion {bIsValid=false, max=<0 0>, min=<0 0>}
    }}
   slot {class=SizeBoxSlot, properties {horizontalAlignment=HAlign_Left}}
  }
  TextBlock_Duration {
   widgetClassPath=&TextBlock
   bInherited=true
   bIsVariable=true
   parent=HorizontalBox_2
   widget=TextBlock_Duration
   properties {
    text=00:30
    font {
     bForceMonospaced=false
     bMaterialIsStencil=false
     fontMaterial=null
     fontObject=&BurbankSmall
     letterSpacing=0
     monospacedWidth=1
     size=18
     skewAmount=0
     typefaceFontName=Bold
     outlineSettings {
      bApplyOutlineToDropShadows=false
      bMiteredCorners=false
      bSeparateFillAlpha=false
      outlineMaterial=null
      outlineSize=0
      outlineColor=<rgba 0 0 0 1>
     }
    }
   }
   slot {class=HorizontalBoxSlot, properties {horizontalAlignment=HAlign_Left, verticalAlignment=VAlign_Center, padding=<ltrb 0 1 15 0>}}
  }
  InputActionWidget {widgetClassPath=&CommonActionWidget, bInherited=true, widget=null}
 ]
}
